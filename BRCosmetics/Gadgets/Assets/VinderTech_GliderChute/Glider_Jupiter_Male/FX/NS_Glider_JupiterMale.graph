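# Niagara system for the thruster exhaust FX on the male Jupiter glider: layered mesh flames, glow sprites, GPU action lines, lights and front/rear mesh trails.
# Scratch modules assign four thruster offsets and sizes, then skeletal mesh sampling pins particles to bones or sockets.
# User ForwardDot and a system integrator value drive spawn, brightness and scale, while system exhaust, length and opacity values feed the trails.
# Front-end preview flags override gameplay values for menu display.
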NiagaraSystem /BRCosmetics/Gadgets/Assets/VinderTech_GliderChute/Glider_Jupiter_Male/FX/NS_Glider_JupiterMale
dependencies {
 usedDataInterfaces []
 usedDynamicInputs [&SetBoolByFloatComparison]
 usedModules [
  &SystemState
  &FN_Integrator
  &self:AwesomeExhaust_Scritch
  &self:AwesomeTrailLength_Scritch
  &self:AwesomeTrailOpacity_Scritch
 ]
}
emitters {
 ActionLines {
  simTarget=GPUComputeSim
  emitterUpdateScript [
   &EmitterState()
   &SpawnRate(
    SpawnRate=&Lerp_Float(A=200, B=300, Alpha=&ClampFloat(Float=$User.ForwardDot))
    Use Spawn Probability=false
   )
  ]
  particleSpawnScript [
   &self:Awesome_Thruster_Scritch(
    BL Thruster Loc:Vector3f=<-45 0 0>
    BR Thruster Loc:Vector3f=<45 0 0>
    FL Thruster Loc:Vector3f=&Multiply_Vector(
     A=<-18 -1 0>
     B=&MakeCustomVectorFromBool(Bool=$User.bIsFrontEndPreview, True Vector=<0.6 0.6 0.6>, False Vector=<1 1 0>)
    )
    FR Thruter Loc:Vector3f=&Multiply_Vector(
     A=<18 1 0>
     B=&MakeCustomVectorFromBool(Bool=$User.bIsFrontEndPreview, True Vector=<0.6 0.6 0.6>, False Vector=<1 1 0>)
    )
    Front Thruster Size:Vector3f=&Multiply_Vector(A=<0.001 0.001 0.05>, B=<0.5 0.5 1.5>)
    Rear Thruster Size:Vector3f=&Multiply_Vector(A=<0.001 0.001 0.05>, B=<0.5 0.5 1.5>)
   ) {module=Awesome_Thruster_Scritch}
   /Niagara/Modules/Spawn/Initialization/InitializeParticle(
    Color=&MultiplyLinearColor(
     Linear Color=<rgba 0 3 10 0.15>
     Scale Factor=&MakeVector4_FromFloat(Float=&UniformAOrBFloat(A=1.3, B=1.6, Distribution Weight=0.2))
    )
    Lifetime=&RandomRangeFloat(Minimum=0.1, Maximum=0.2)
    Module.Write Color=true
    Module.Write Lifetime=true
    Module.Write Mass=false
    Module.Write Material Random=false
    Module.Write Position=false
    Module.Write Scale=true
    Module.Write SpriteRotation=false
    Module.Write SpriteSize=true
    Position=&SimulationPosition()
    Sprite Size=<1 20>
   )
   &SphereLocation(
    MaskSpawnGroup=false
    Override Seed=false
    Sphere Origin:NiagaraPosition=<0 0 0>
    Sphere Radius=&RandomRangeFloat(Minimum=0, Maximum=6)
    Surface Only=false
    UseNonUniformScale=false
    UseOffset=false
    UseSphereOrientation=false
   )
   &SkeletalMeshLocation(
    Apply Sampled Orientation to Mesh Particles=true
    Bone / Socket Index=$Particles.ThrusterSpawnGroup
    Bone Sampling Mode=::NewEnumerator4
    Inherit Velocity=false
    Interpolate Mesh Sampling=false
    MaskSpawnGroup=false
    Sampled Position Offset=&Add_Vector(A=$Particles.Thruster Positions, B=Linked() {linkedVariable {name=Output.SphereLocation.Position, type=NiagaraPosition}})
    Sampled Position Offset Coordinate Space:ENiagaraMeshTransforms=::NewEnumerator3
    Skeletal Mesh=$User.SkeletalMesh
    Use Sampled Mesh Attribute as Ribbon ID=false
   )
   &AddVelocity(Velocity=&MakeCustomVectorFromBool(
     Bool=&SetBoolByIntComparison(
      A=Linked() {linkedVariable {name=Particles.F/B Thrusters, type=NiagaraInt32, defaultValue=~, description=~}}
      B=0
      Comparison Type=::NewEnumerator4
     )
     True Vector=<0 200 40>
     False Vector=<0 200 0>
    ))
  ]
  particleUpdateScript [
   &SolveForcesAndVelocity(Clamp Velocity=false, Limit Acceleration=false)
   &ParticleState()
   /SetVariables(Particles.MeshOrientation=&LerpQuaternion(
     Quaternion A:Quat4f=&MultiplyQuaternion(
      Quaternion A=$Output.SkeletalMeshLocation.SampledRotation
      Quaternion B=&OrientMeshToVector(Look At Direction=<0 0 1>, Reference Axis Vector=<1 0 0>)
     )
     Quaternion B:Quat4f=&MultiplyQuaternion(
      Quaternion A=$Output.SkeletalMeshLocation.SampledRotation
      Quaternion B=&OrientMeshToVector(Look At Direction=<0 0 -1>, Reference Axis Vector=<1 0 0>)
     )
     Lerp Factor:NiagaraFloat=&ModuloFloat(A=&MakeFloatFromInt(Int=$Particles.ThrusterSpawnGroup), B=2)
     Enable Spherical Linear Interpolation:NiagaraBool=false
    )) {isSetParameters=true, module=SetVariables_#1}
   &ScaleColor(
    Scale Alpha=&Lerp_Float(
     A=1
     B=&Lerp_Float(A=1, B=1.1, Alpha=&ClampFloat(Float=$User.ForwardDot))
     Alpha=$Particles.NormalizedAge
    )
    Scale RGB=&Lerp_Vector(A=<1 1 1>, B=<2 2 2>, Alpha=&ClampFloat(Float=$User.ForwardDot, Min=-0.15))
    ScaleA=true
    ScaleRGB=true
   )
  ]
  rendererClasses [&NiagaraSpriteRendererProperties]
  renderers [{rendererClass=&NiagaraSpriteRendererProperties}]
 }
 ExhaustFlameInner001 {
  emitterUpdateScript [
   &EmitterState()
   &SpawnRate(
    SpawnRate=&Lerp_Float(A=16, B=32, Alpha=&ClampFloat(Float=$User.ForwardDot))
    Use Spawn Probability=false
   )
  ]
  particleSpawnScript [
   &self:Awesome_Thruster_Scritch(
    BL Thruster Loc:Vector3f=<-54 0 0>
    BR Thruster Loc:Vector3f=<54 0 0>
    FL Thruster Loc:Vector3f=<-24 0 0>
    FR Thruter Loc:Vector3f=<24 0 0>
    Front Thruster Size:Vector3f=&Multiply_Vector(A=<0.016 0.016 0.05>, B=<1 1 0.1>)
    Rear Thruster Size:Vector3f=&Multiply_Vector(A=<0.016 0.016 0.05>, B=<1 1 0.1>)
   ) {module=Awesome_Thruster_Scritch}
   &InitializeParticle(
    Color=<rgba 10 10 10 1>
    Lifetime=&Lerp_Float(A=1, B=0.5, Alpha=$User.ForwardDot)
    Mesh Scale=$Particles.Thruster Scale
    Module.Write Color=true
    Module.Write Lifetime=true
    Module.Write Mass=false
    Module.Write Material Random=false
    Module.Write Position=false
    Module.Write Scale=true
    Module.Write SpriteRotation=false
    Module.Write SpriteSize=false
    Position=&SimulationPosition()
    Sprite Size=<5 5>
   )
   /Niagara/Modules/Spawn/Orientation/V2/InitialMeshOrientation(Orientation Axis:Vector3f=<0 0 0>, Use Orientation Vector=true, Use Rotation Vector=false) {enabled=false}
   /Niagara/Modules/Spawn/Orientation/V3/InitialMeshOrientation(Orientation Axis=<0 0 0>, Use Orientation Vector=true, Use Rotation Vector=false) {module=InitialMeshOrientation001}
   &DynamicMaterialParameters(
    Index 0 Param 1=&RandomRangeFloat(Minimum=0, Maximum=1)
    Module.Param0WriteEnabled=true
    Module.Param1WriteEnabled=false
    Module.Param2WriteEnabled=false
    Module.Param3WriteEnabled=false
   )
  ]
  particleUpdateScript [
   &ParticleState()
   &SkeletalMeshLocation(
    Apply Sampled Orientation to Mesh Particles=true
    Bone / Socket Index=$Particles.ThrusterSpawnGroup
    Bone Sampling Mode=::NewEnumerator4
    Inherit Velocity=false
    Interpolate Mesh Sampling=false
    MaskSpawnGroup=false
    Sampled Position Offset=$Particles.Thruster Positions
    Sampled Position Offset Coordinate Space:ENiagaraMeshTransforms=::NewEnumerator3
    Skeletal Mesh=$User.SkeletalMesh
    Use Sampled Mesh Attribute as Ribbon ID=false
   )
   /SetVariables(Particles.MeshOrientation=&LerpQuaternion(
     Quaternion A:Quat4f=&MultiplyQuaternion(
      Quaternion A=$Output.SkeletalMeshLocation.SampledRotation
      Quaternion B=&OrientMeshToVector(Look At Direction=<0 0 1>, Reference Axis Vector=<1 0 0>)
     )
     Quaternion B:Quat4f=&MultiplyQuaternion(
      Quaternion A=$Output.SkeletalMeshLocation.SampledRotation
      Quaternion B=&OrientMeshToVector(Look At Direction=<0 0 -1>, Reference Axis Vector=<1 0 0>)
     )
     Lerp Factor:NiagaraFloat=&ModuloFloat(A=&MakeFloatFromInt(Int=$Particles.ThrusterSpawnGroup), B=2)
     Enable Spherical Linear Interpolation:NiagaraBool=false
    )) {isSetParameters=true, module=SetVariables_#1}
   &ScaleColor(
    Scale Alpha=&FN_FrontEndFloatModifier(
     Gameplay=&Lerp_Float(
      A=1
      B=&Lerp_Float(A=1, B=1.1, Alpha=&ClampFloat(Float=$User.ForwardDot))
      Alpha=$Particles.NormalizedAge
     )
     Front End=2
    )
    Scale RGB=&Lerp_Vector(A=<5 5 15>, B=<2 2 2>, Alpha=&ClampFloat(Float=$User.ForwardDot, Min=-15))
    ScaleA=true
    ScaleRGB=true
   )
  ]
  rendererClasses [&NiagaraMeshRendererProperties]
  renderers [{rendererClass=&NiagaraMeshRendererProperties, rendererIndex=0}]
 }
 ExhaustFlame_Big {
  emitterUpdateScript [
   &EmitterState()
   &SpawnRate(
    SpawnRate=&Lerp_Float(A=16, B=32, Alpha=&ClampFloat(Float=$User.ForwardDot))
    Use Spawn Probability=false
   )
  ]
  particleSpawnScript [
   &self:Awesome_Thruster_Scritch(
    BL Thruster Loc:Vector3f=<-45 0 0>
    BR Thruster Loc:Vector3f=<45 0 0>
    FL Thruster Loc:Vector3f=<-15 0 0>
    FR Thruter Loc:Vector3f=<15 0 0>
    Front Thruster Size:Vector3f=<0.052 0.052 0.2>
    Rear Thruster Size:Vector3f=<0.058 0.058 0.2>
   ) {module=Awesome_Thruster_Scritch}
   &InitializeParticle(
    Color=<rgba 1 1 1 0.3>
    Lifetime=&Lerp_Float(A=1, B=0.5, Alpha=$User.ForwardDot)
    Mesh Scale=$Particles.Thruster Scale
    Module.Write Color=true
    Module.Write Lifetime=true
    Module.Write Mass=false
    Module.Write Material Random=false
    Module.Write Position=false
    Module.Write Scale=true
    Module.Write SpriteRotation=false
    Module.Write SpriteSize=false
    Position=&SimulationPosition()
    Sprite Size=<5 5>
   )
   /Niagara/Modules/Spawn/Orientation/V2/InitialMeshOrientation(
    Orientation Axis:Vector3f=&Lerp_Vector(
     A=<0 0 -1>
     B=<0 0 1>
     Alpha=&MakeFloatFromInt(Int=Linked() {linkedVariable {name=Particles.L/R Thrusters, type=NiagaraInt32, defaultValue=~, description=~}})
    )
    Use Orientation Vector=true
    Use Rotation Vector=false
   ) {enabled=false}
   /Niagara/Modules/Spawn/Orientation/V3/InitialMeshOrientation(
    Orientation Axis=&Lerp_Vector(
     A=<0 0 -1>
     B=<0 0 1>
     Alpha=&MakeFloatFromInt(Int=Linked() {linkedVariable {name=Particles.L/R Thrusters, type=NiagaraInt32, defaultValue=~, description=~}})
    )
    Use Orientation Vector=true
    Use Rotation Vector=false
   ) {module=InitialMeshOrientation001}
   &DynamicMaterialParameters(
    Index 0 Param 1=&RandomRangeFloat(Minimum=0, Maximum=1)
    Index 0 Param 4=&Lerp_Float(A=2, B=3, Alpha=$User.ForwardDot)
    Module.Param0WriteEnabled=true
    Module.Param1WriteEnabled=false
    Module.Param2WriteEnabled=false
    Module.Param3WriteEnabled=true
   )
  ]
  particleUpdateScript [
   &ParticleState()
   &SkeletalMeshLocation(
    Apply Sampled Orientation to Mesh Particles=true
    Bone / Socket Index=$Particles.ThrusterSpawnGroup
    Bone Sampling Mode=::NewEnumerator4
    Inherit Velocity=false
    Interpolate Mesh Sampling=false
    MaskSpawnGroup=false
    Sampled Position Offset=$Particles.Thruster Positions
    Sampled Position Offset Coordinate Space:ENiagaraMeshTransforms=::NewEnumerator3
    Skeletal Mesh=$User.SkeletalMesh
    Use Sampled Mesh Attribute as Ribbon ID=false
   )
   &ScaleColor(
    Scale RGB=&MakeCustomVectorFromBool(
     Bool=$User.bIsFrontEndPreview
     True Vector=<3 3 3>
     False Vector=&Lerp_Vector(A=<1 1 1>, B=<2 2 2>, Alpha=$System.FN_Integrator.Value)
    )
    ScaleA=false
    ScaleRGB=true
   )
   &ScaleMeshSize(Scale Factor=&FN_FrontEndVector3Modifier(Gameplay=&Lerp_Vector(A=<1 1 1>, B=<1.15 1.15 1.2>, Alpha=$System.FN_Integrator.Value), Front End=<1.05 1.05 1.2>))
  ]
  rendererClasses [&NiagaraMeshRendererProperties]
  renderers [{rendererClass=&NiagaraMeshRendererProperties, rendererIndex=0}]
 }
 ExhaustFlame_Small {
  emitterUpdateScript [
   &EmitterState()
   &SpawnRate(
    SpawnRate=&FN_FrontEndFloatModifier(Gameplay=&Lerp_Float(A=16, B=32, Alpha=$User.ForwardDot), Front End=32)
    Use Spawn Probability=false
   )
  ]
  particleSpawnScript [
   &self:Awesome_Thruster_Scritch(
    BL Thruster Loc:Vector3f=<-30 0 0>
    BR Thruster Loc:Vector3f=<30 0 0>
    FL Thruster Loc:Vector3f=<-5 0 0>
    FR Thruter Loc:Vector3f=<5 0 0>
    Front Thruster Size:Vector3f=<0.0225 0.0225 0.3>
    Rear Thruster Size:Vector3f=<0.0275 0.0275 0.3>
   ) {module=Awesome_Thruster_Scritch}
   &InitializeParticle(
    Color=<rgba 1 1 3 0.8>
    Lifetime=&FN_FrontEndFloatModifier(Gameplay=&Lerp_Float(A=1, B=0.5, Alpha=$User.ForwardDot), Front End=0.5)
    Mesh Scale=$Particles.Thruster Scale
    Module.Write Color=true
    Module.Write Lifetime=true
    Module.Write Mass=false
    Module.Write Material Random=false
    Module.Write Position=false
    Module.Write Scale=true
    Module.Write SpriteRotation=false
    Module.Write SpriteSize=false
    Position=&SimulationPosition()
    Sprite Size=<5 5>
   )
   /Niagara/Modules/Spawn/Orientation/V2/InitialMeshOrientation(Orientation Axis:Vector3f=<0 0 0>, Use Orientation Vector=true, Use Rotation Vector=false) {enabled=false}
   /Niagara/Modules/Spawn/Orientation/V3/InitialMeshOrientation(Orientation Axis=<0 0 0>, Use Orientation Vector=true, Use Rotation Vector=false) {module=InitialMeshOrientation001}
   &DynamicMaterialParameters(
    Index 0 Param 1=&RandomRangeFloat(Minimum=0, Maximum=1)
    Index 0 Param 4=&Lerp_Float(A=2, B=3, Alpha=$User.ForwardDot)
    Module.Param0WriteEnabled=true
    Module.Param1WriteEnabled=false
    Module.Param2WriteEnabled=false
    Module.Param3WriteEnabled=true
   )
  ]
  particleUpdateScript [
   &ParticleState()
   &SkeletalMeshLocation(
    Apply Sampled Orientation to Mesh Particles=true
    Bone / Socket Index=$Particles.ThrusterSpawnGroup
    Bone Sampling Mode=::NewEnumerator4
    Inherit Velocity=false
    Interpolate Mesh Sampling=false
    MaskSpawnGroup=false
    Sampled Position Offset=$Particles.Thruster Positions
    Sampled Position Offset Coordinate Space:ENiagaraMeshTransforms=::NewEnumerator3
    Skeletal Mesh=$User.SkeletalMesh
    Use Sampled Mesh Attribute as Ribbon ID=false
   )
   /SetVariables(Particles.MeshOrientation=&LerpQuaternion(
     Quaternion A:Quat4f=&MultiplyQuaternion(
      Quaternion A=$Output.SkeletalMeshLocation.SampledRotation
      Quaternion B=&OrientMeshToVector(Look At Direction=<0 0 1>, Reference Axis Vector=<1 0 0>)
     )
     Quaternion B:Quat4f=&MultiplyQuaternion(
      Quaternion A=$Output.SkeletalMeshLocation.SampledRotation
      Quaternion B=&OrientMeshToVector(Look At Direction=<0 0 -1>, Reference Axis Vector=<1 0 0>)
     )
     Lerp Factor:NiagaraFloat=&ModuloFloat(A=&MakeFloatFromInt(Int=$Particles.ThrusterSpawnGroup), B=2)
     Enable Spherical Linear Interpolation:NiagaraBool=false
    )) {isSetParameters=true, module=SetVariables_#1}
   &ScaleColor(
    Scale RGB=&FN_FrontEndVector3Modifier(Gameplay=&Lerp_Vector(A=<1 1 1>, B=<2 2 2>, Alpha=$System.FN_Integrator.Value), Front End=<2 2 2>)
    ScaleA=false
    ScaleRGB=true
   )
   &ScaleMeshSize(Scale Factor=&FN_FrontEndVector3Modifier(Gameplay=&Lerp_Vector(A=<1 1 1>, B=<1.3 1.3 1.4>, Alpha=$System.FN_Integrator.Value), Front End=<1.2 1.2 1.4>))
  ]
  rendererClasses [&NiagaraMeshRendererProperties]
  renderers [{rendererClass=&NiagaraMeshRendererProperties, rendererIndex=0}]
 }
 Glow {
  emitterUpdateScript [
   &EmitterState(
    Inactive Response=::NewEnumerator1
    Life Cycle Mode=::NewEnumerator1
    Loop Behavior=::NewEnumerator1
    Loop Duration Mode=::NewEnumerator1
    UseLoopDelay=false
   )
   &SpawnBurst_Instantaneous(
    Spawn Count=4
    Spawn Probability=&MakeCustomFloatFromBool(
     Bool=&SetBoolByIntComparison(A=$System.LoopCount, B=0, Comparison Type=::NewEnumerator4)
     True Float=1
     False Float=0
    )
    Use Spawn Probability=true
   )
  ]
  particleSpawnScript [
   &self:Awesome_Thruster_Scritch(BL Thruster Loc:Vector3f=<-55 0 0>, BR Thruster Loc:Vector3f=<55 0 0>, FL Thruster Loc:Vector3f=<-25 0 0>, FR Thruter Loc:Vector3f=<25 0 0>) {module=Awesome_Thruster_Scritch}
   &InitializeParticle(
    Color=<rgba 0.2 0.5 1 0.35>
    Lifetime=5
    Module.Write Color=true
    Module.Write Lifetime=true
    Module.Write Mass=true
    Module.Write Material Random=true
    Module.Write Position=false
    Module.Write Scale=false
    Module.Write SpriteRotation=false
    Module.Write SpriteSize=true
    Position=&SimulationPosition()
    Sprite Size=&Vector2DFromFloat(Value=65)
   )
   &CameraOffset(Camera Offset Amount=20)
   &DynamicMaterialParameters(
    Index 0 Param 1=1
    Module.Param0WriteEnabled=true
    Module.Param0WriteEnabled1=true
    Module.Param1WriteEnabled=true
    Module.Param1WriteEnabled1=true
    Module.Param2WriteEnabled=true
    Module.Param2WriteEnabled1=true
    Module.Param3WriteEnabled=true
    Module.Param3WriteEnabled1=true
    Write Parameter Index 1=true
   )
  ]
  particleUpdateScript [
   &ParticleState(
    Kill Particles When Lifetime Has Elapsed=false
    Let Infinitely Lived Particles Die When Emitter Deactivates=true
    Loop Particles Lifetime=true
   )
   &SkeletalMeshLocation(
    Bone / Socket Index=$Particles.ThrusterSpawnGroup
    Bone Sampling Mode=::NewEnumerator4
    Inherit Velocity=false
    Interpolate Mesh Sampling=false
    MaskSpawnGroup=false
    Sampled Position Offset=$Particles.Thruster Positions
    Sampled Position Offset Coordinate Space:ENiagaraMeshTransforms=::NewEnumerator3
    Skeletal Mesh=$User.SkeletalMesh
    Use Sampled Mesh Attribute as Ribbon ID=false
   )
   &SolveForcesAndVelocity(Clamp Velocity=false, Limit Acceleration=false)
   &ScaleSpriteSize(Scale Factor=&Vector2DFromFloat(Value=&Lerp_Float(A=1, B=1.2, Alpha=$System.FN_Integrator.Value)))
   &ScaleColor(Scale Alpha=&Lerp_Float(A=1, B=1.1, Alpha=$System.FN_Integrator.Value), ScaleA=true, ScaleRGB=true)
  ]
  rendererClasses [&NiagaraSpriteRendererProperties]
  renderers [{rendererClass=&NiagaraSpriteRendererProperties}]
 }
 InnerGlow {
  emitterUpdateScript [
   &EmitterState()
   &SpawnBurst_Instantaneous(
    Spawn Count=4
    Spawn Probability=&MakeCustomFloatFromBool(
     Bool=&SetBoolByIntComparison(A=0, B=$Emitter.LoopCount, Comparison Type=::NewEnumerator4)
     True Float=1
     False Float=0
    )
    Use Spawn Probability=true
   )
  ]
  particleSpawnScript [
   &self:Awesome_Thruster_Scritch(
    BL Thruster Loc:Vector3f=<-45 0 0>
    BR Thruster Loc:Vector3f=<45 0 0>
    FL Thruster Loc:Vector3f=<-18 0 0>
    FR Thruter Loc:Vector3f=<18 0 0>
    Front Thruster Size:Vector3f=&Multiply_Vector(A=<0.001 0.001 0.05>, B=<0.5 0.5 1.5>)
    Rear Thruster Size:Vector3f=&Multiply_Vector(A=<0.001 0.001 0.05>, B=<0.5 0.5 1.5>)
   ) {module=Awesome_Thruster_Scritch}
   &InitializeParticle(
    Color=<rgba 0 0.1 1 1>
    Module.Write Color=true
    Module.Write Lifetime=true
    Module.Write Mass=false
    Module.Write Material Random=false
    Module.Write Position=false
    Module.Write Scale=true
    Module.Write SpriteRotation=false
    Module.Write SpriteSize=true
    Position=&SimulationPosition()
    Sprite Size=<20 20>
   )
  ]
  particleUpdateScript [
   &SolveForcesAndVelocity(Clamp Velocity=false, Limit Acceleration=false)
   &ParticleState(
    Kill Particles When Lifetime Has Elapsed=false
    Let Infinitely Lived Particles Die When Emitter Deactivates=true
    Loop Particles Lifetime=true
   )
   &SkeletalMeshLocation(
    Apply Sampled Orientation to Mesh Particles=true
    Bone / Socket Index=$Particles.ThrusterSpawnGroup
    Bone Sampling Mode=::NewEnumerator4
    Inherit Velocity=false
    Interpolate Mesh Sampling=false
    MaskSpawnGroup=false
    Sampled Position Offset=&Add_Vector(A=$Particles.Thruster Positions, B=$Output.SphereLocation.Position)
    Sampled Position Offset Coordinate Space:ENiagaraMeshTransforms=::NewEnumerator3
    Skeletal Mesh=$User.SkeletalMesh
    Use Sampled Mesh Attribute as Ribbon ID=false
   )
   /SetVariables(Particles.MeshOrientation=&LerpQuaternion(
     Quaternion A:Quat4f=&MultiplyQuaternion(
      Quaternion A=$Output.SkeletalMeshLocation.SampledRotation
      Quaternion B=&OrientMeshToVector(Look At Direction=<0 0 1>, Reference Axis Vector=<1 0 0>)
     )
     Quaternion B:Quat4f=&MultiplyQuaternion(
      Quaternion A=$Output.SkeletalMeshLocation.SampledRotation
      Quaternion B=&OrientMeshToVector(Look At Direction=<0 0 -1>, Reference Axis Vector=<1 0 0>)
     )
     Lerp Factor:NiagaraFloat=&ModuloFloat(A=&MakeFloatFromInt(Int=$Particles.ThrusterSpawnGroup), B=2)
     Enable Spherical Linear Interpolation:NiagaraBool=false
    )) {isSetParameters=true, module=SetVariables_#1}
   &ScaleColor(
    Scale Alpha=&Lerp_Float(
     A=1
     B=&Lerp_Float(A=1, B=1.1, Alpha=&ClampFloat(Float=$User.ForwardDot))
     Alpha=$Particles.NormalizedAge
    )
    Scale RGB=&MakeCustomVectorFromBool(
     Bool=$User.bIsFrontEndPreview
     True Vector=<3 3 3>
     False Vector=&Lerp_Vector(A=<1 1 1>, B=<2 2 2>, Alpha=&ClampFloat(Float=$User.ForwardDot, Min=-0.15))
    )
    ScaleA=true
    ScaleRGB=true
   )
  ]
  rendererClasses [&NiagaraSpriteRendererProperties]
  renderers [{rendererClass=&NiagaraSpriteRendererProperties}]
 }
 Lights {
  emitterUpdateScript [
   &EmitterState(Life Cycle Mode=::NewEnumerator1, Loop Behavior=::NewEnumerator1, Loop Duration=1, UseLoopDelay=false)
   &SpawnBurst_Instantaneous(Spawn Count=4, Use Spawn Probability=false)
  ]
  particleSpawnScript [
   &self:Awesome_Thruster_Scritch(BL Thruster Loc:Vector3f=<-80 0 0>, BR Thruster Loc:Vector3f=<80 0 0>, FL Thruster Loc:Vector3f=<-40 0 0>, FR Thruter Loc:Vector3f=<40 0 0>) {module=Awesome_Thruster_Scritch}
   &InitializeParticle(
    Color=<rgba 16 24 32 1>
    Module.Write Color=true
    Module.Write Lifetime=true
    Module.Write Mass=false
    Module.Write Material Random=false
    Module.Write Position=false
    Module.Write Scale=false
    Module.Write SpriteRotation=false
    Module.Write SpriteSize=true
    Position=&SimulationPosition()
    Sprite Size=<5 5>
   )
  ]
  particleUpdateScript [
   &ParticleState(Kill Particles When Lifetime Has Elapsed=false)
   /SetVariables(Particles.LightExponent:NiagaraFloat=&Lerp_Float(A=3, B=3.3, Alpha=&Sine(Normalized Angle=$Emitter.Age, Period=2, Scale=0.5, Bias=1))) {isSetParameters=true}
   &SkeletalMeshLocation(
    Bone / Socket Index=$Particles.ThrusterSpawnGroup
    Bone Sampling Mode=::NewEnumerator4
    Inherit Velocity=false
    Interpolate Mesh Sampling=false
    MaskSpawnGroup=false
    Sampled Position Offset=$Particles.Thruster Positions
    Sampled Position Offset Coordinate Space:ENiagaraMeshTransforms=::NewEnumerator3
    Skeletal Mesh=$User.SkeletalMesh
    Use Sampled Mesh Attribute as Ribbon ID=false
   )
   &ScaleColor(
    Scale Alpha=&Lerp_Float(A=0, B=&Lerp_Float(A=2, B=4, Alpha=$User.ForwardDot), Alpha=$Particles.NormalizedAge)
    Scale RGB=&Lerp_Vector(
     A=&VectorFromFloat(Value=1)
     B=&VectorFromFloat(Value=1.25)
     Alpha=$User.ForwardDot
    )
    ScaleA=false
    ScaleRGB=true
   )
  ]
  rendererClasses [&NiagaraLightRendererProperties]
  renderers [{rendererClass=&NiagaraLightRendererProperties, rendererIndex=0}]
 }
 MeshTrails_Front {
  emitterUpdateScript [
   &EmitterState()
   &SpawnBurst_Instantaneous(
    Spawn Count=2
    Spawn Probability=&MakeCustomFloatFromBool(
     Bool=&SetBoolByIntComparison(A=$System.LoopCount, B=0, Comparison Type=::NewEnumerator4)
     True Float=1
     False Float=0
    )
    Use Spawn Probability=true
   ) {module=SpawnBurst_Instantaneous001}
  ]
  particleSpawnScript [
   &self:Awesome_Thruster_Scritch(
    BL Thruster Loc:Vector3f=<-30 0 0>
    BR Thruster Loc:Vector3f=<30 0 0>
    FL Thruster Loc:Vector3f=<-5 0 0>
    FR Thruter Loc:Vector3f=<5 0 0>
    Front Thruster Size:Vector3f=<0.0225 0.0225 0.3>
    Rear Thruster Size:Vector3f=<0.0275 0.0275 0.3>
   ) {module=Awesome_Thruster_Scritch}
   &InitializeParticle(
    Color=<rgba 1 1 1 0.035>
    Lifetime=5
    Module.Write Color=true
    Module.Write Lifetime=true
    Module.Write Mass=false
    Module.Write Material Random=true
    Module.Write Position=false
    Module.Write Scale=false
    Module.Write SpriteRotation=false
    Module.Write SpriteSize=false
   )
   &self:Mesh_Trail_Setup(Trail Length=300, Trail Width=100) {module=Mesh_Trail_Setup}
  ]
  particleUpdateScript [
   &ParticleState(Kill Particles When Lifetime Has Elapsed=false, Let Infinitely Lived Particles Die When Emitter Deactivates=true)
   &SkeletalMeshLocation(
    Apply To Particle Position=false
    Inherit Velocity=false
    Interpolate Mesh Sampling=false
    MaskSpawnGroup=false
    Mesh Sampling Type=::NewEnumerator5
    Skeletal Mesh=$User.SkeletalMesh
    Socket Sampling Mode=::NewEnumerator5
    Write Sampled Mesh Attrributes to Particle Attributes=true
   )
   &SkeletalMeshLocation(
    Apply To Particle Position=false
    Bone / Socket Index=1
    Inherit Velocity=false
    Interpolate Mesh Sampling=false
    MaskSpawnGroup=false
    Mesh Sampling Type=::NewEnumerator5
    Skeletal Mesh=$User.SkeletalMesh
    Socket Sampling Mode=::NewEnumerator5
    Write Sampled Mesh Attrributes to Particle Attributes=true
   ) {module=SkeletalMeshLocation001}
   /SetVariables(
    Particles.Trail_FL_Orientation:Quat4f=&MultiplyQuaternion(
     Quaternion A=&MultiplyQuaternion(
      Quaternion A=$Output.SkeletalMeshLocation.SampledRotation
      Quaternion B=&MakeQuatFromAxisAngle(Axis=<0 1 0>, AngleInDegrees=-90)
     )
     Quaternion B=&MakeQuatFromAxisAngle(Axis=<1 0 0>, AngleInDegrees=90)
    )
    Particles.Trail_FL_Position:Vector3f=&Add_Vector(
     A=$Output.SkeletalMeshLocation.SampledPosition
     B=&RotateVectorByQuaternion(Vector To Rotate=<-47 0 -18>, Quaternion=$Output.SkeletalMeshLocation.SampledRotation)
    )
    Particles.Trail_FR_Orientation:Quat4f=&MultiplyQuaternion(
     Quaternion A=&MultiplyQuaternion(
      Quaternion A=$Output.SkeletalMeshLocation001.SampledRotation
      Quaternion B=&MakeQuatFromAxisAngle(Axis=<0 1 0>, AngleInDegrees=-91)
     )
     Quaternion B=&MakeQuatFromAxisAngle(Axis=<1 0 0>, AngleInDegrees=90)
    )
    Particles.Trail_FR_Position:Vector3f=&Add_Vector(
     A=$Output.SkeletalMeshLocation001.SampledPosition
     B=&RotateVectorByQuaternion(Vector To Rotate=<-47 0 -18>, Quaternion=$Output.SkeletalMeshLocation001.SampledRotation)
    )
   ) {isSetParameters=true, module=SetVariables_#2}
   &self:Exhaust_to_DynaMat(Exhaust to DynaMat:Vector3f=$System.CurrentExhaust, SpeedMult=13) {module=Exhaust_to_DynaMat}
   &self:Mesh_Trail_Length_Update(Dynamic Length Scale:NiagaraFloat=&MakeCustomFloatFromBool(
     Bool=&Boolean_OrOperation(Bool1=$User.bIsFrontEnd, Bool2=$User.bIsFrontEndPreview)
     True Float=1
     False Float=$System.TrailLength
    )) {module=Mesh_Trail_Length_Update}
   &ScaleColor(
    Scale Alpha=&MakeCustomFloatFromBool(Bool=$User.bIsFrontEndPreview, True Float=1, False Float=$System.TrailOpacity)
    ScaleA=true
    ScaleRGB=false
   )
  ]
  rendererClasses [&NiagaraMeshRendererProperties]
  renderers [
   {rendererClass=&NiagaraMeshRendererProperties, rendererIndex=0}
   {rendererClass=&NiagaraMeshRendererProperties, rendererIndex=1}
  ]
 }
 MeshTrails_Rear {
  emitterUpdateScript [
   &EmitterState()
   &SpawnBurst_Instantaneous(
    Spawn Count=2
    Spawn Probability=&MakeCustomFloatFromBool(
     Bool=&SetBoolByIntComparison(A=$System.LoopCount, B=0, Comparison Type=::NewEnumerator4)
     True Float=1
     False Float=0
    )
    Use Spawn Probability=true
   ) {module=SpawnBurst_Instantaneous001}
  ]
  particleSpawnScript [
   &self:Awesome_Thruster_Scritch(
    BL Thruster Loc:Vector3f=<-30 0 0>
    BR Thruster Loc:Vector3f=<30 0 0>
    FL Thruster Loc:Vector3f=<-5 0 0>
    FR Thruter Loc:Vector3f=<5 0 0>
    Front Thruster Size:Vector3f=<0.0225 0.0225 0.3>
    Rear Thruster Size:Vector3f=<0.0275 0.0275 0.3>
   ) {module=Awesome_Thruster_Scritch}
   &InitializeParticle(
    Color=<rgba 1 1 1 0.035>
    Lifetime=5
    Module.Write Color=true
    Module.Write Lifetime=true
    Module.Write Mass=false
    Module.Write Material Random=true
    Module.Write Position=false
    Module.Write Scale=false
    Module.Write SpriteRotation=false
    Module.Write SpriteSize=false
   )
   &self:Mesh_Trail_Setup(Trail Length=200, Trail Width=25) {module=Mesh_Trail_Setup}
  ]
  particleUpdateScript [
   &ParticleState(Kill Particles When Lifetime Has Elapsed=false, Let Infinitely Lived Particles Die When Emitter Deactivates=true)
   &SkeletalMeshLocation(
    Apply To Particle Position=false
    Bone / Socket Index=2
    Inherit Velocity=false
    Interpolate Mesh Sampling=false
    MaskSpawnGroup=false
    Mesh Sampling Type=::NewEnumerator5
    Skeletal Mesh=$User.SkeletalMesh
    Socket Sampling Mode=::NewEnumerator5
    Write Sampled Mesh Attrributes to Particle Attributes=true
   )
   &SkeletalMeshLocation(
    Apply To Particle Position=false
    Bone / Socket Index=3
    Inherit Velocity=false
    Interpolate Mesh Sampling=false
    MaskSpawnGroup=false
    Mesh Sampling Type=::NewEnumerator5
    Skeletal Mesh=$User.SkeletalMesh
    Socket Sampling Mode=::NewEnumerator5
    Write Sampled Mesh Attrributes to Particle Attributes=true
   ) {module=SkeletalMeshLocation001}
   /SetVariables(
    Particles.Trail_FL_Orientation:Quat4f=&MultiplyQuaternion(
     Quaternion A=&MultiplyQuaternion(
      Quaternion A=$Output.SkeletalMeshLocation.SampledRotation
      Quaternion B=&MakeQuatFromAxisAngle(Axis=<0 1 0>, AngleInDegrees=-90)
     )
     Quaternion B=&MakeQuatFromAxisAngle(Axis=<1 0 0>, AngleInDegrees=90)
    )
    Particles.Trail_FL_Position:Vector3f=&Add_Vector(
     A=$Output.SkeletalMeshLocation.SampledPosition
     B=&RotateVectorByQuaternion(Vector To Rotate=<-20 0 -15>, Quaternion=$Output.SkeletalMeshLocation.SampledRotation)
    )
    Particles.Trail_FR_Orientation:Quat4f=&MultiplyQuaternion(
     Quaternion A=&MultiplyQuaternion(
      Quaternion A=$Output.SkeletalMeshLocation001.SampledRotation
      Quaternion B=&MakeQuatFromAxisAngle(Axis=<0 1 0>, AngleInDegrees=-90)
     )
     Quaternion B=&MakeQuatFromAxisAngle(Axis=<1 0 0>, AngleInDegrees=90)
    )
    Particles.Trail_FR_Position:Vector3f=&Add_Vector(
     A=$Output.SkeletalMeshLocation001.SampledPosition
     B=&RotateVectorByQuaternion(Vector To Rotate=<-20 0 -15>, Quaternion=$Output.SkeletalMeshLocation001.SampledRotation)
    )
   ) {isSetParameters=true, module=SetVariables_#2}
   &self:Exhaust_to_DynaMat(Exhaust to DynaMat:Vector3f=$System.CurrentExhaust, SpeedMult=18) {module=Exhaust_to_DynaMat}
   &self:Mesh_Trail_Length_Update(Dynamic Length Scale:NiagaraFloat=&MakeCustomFloatFromBool(
     Bool=&Boolean_OrOperation(Bool1=$User.bIsFrontEnd, Bool2=$User.bIsFrontEndPreview)
     True Float=1
     False Float=$System.TrailLength
    )) {module=Mesh_Trail_Length_Update}
   &ScaleColor(
    Scale Alpha=&MakeCustomFloatFromBool(Bool=$User.bIsFrontEndPreview, True Float=1, False Float=$System.TrailOpacity)
    ScaleA=true
    ScaleRGB=false
   )
  ]
  rendererClasses [&NiagaraMeshRendererProperties]
  renderers [
   {rendererClass=&NiagaraMeshRendererProperties, rendererIndex=0}
   {rendererClass=&NiagaraMeshRendererProperties, rendererIndex=1}
  ]
 }
}
summary {
 systemName=NS_Glider_JupiterMale
 emitters [
  {emitterName=MeshTrails_Front, rendererClasses [&NiagaraMeshRendererProperties]}
  {emitterName=MeshTrails_Rear, rendererClasses [&NiagaraMeshRendererProperties]}
  {emitterName=ExhaustFlame_Small, rendererClasses [&NiagaraMeshRendererProperties]}
  {emitterName=ExhaustFlame_Big, rendererClasses [&NiagaraMeshRendererProperties]}
  {emitterName=Lights, rendererClasses [&NiagaraLightRendererProperties]}
  {emitterName=Glow, rendererClasses [&NiagaraSpriteRendererProperties]}
  {emitterName=ExhaustFlameInner001, rendererClasses [&NiagaraMeshRendererProperties]}
  {emitterName=ActionLines, simTarget=GPUComputeSim, rendererClasses [&NiagaraSpriteRendererProperties]}
  {emitterName=InnerGlow, rendererClasses [&NiagaraSpriteRendererProperties]}
 ]
 userVariables [
  User.bIsFrontEnd {type=NiagaraBool}
  User.bIsFrontEndPreview {type=NiagaraBool, defaultValue=true}
  User.ForwardDot {type=NiagaraFloat, defaultValue=0, description=""}
  User.RightDot {type=NiagaraFloat, defaultValue=0, description=""}
  User.SkeletalMesh {
   type=NiagaraDataInterfaceSkeletalMesh
   defaultValue {
    struct=NiagaraExt_VariableValue_DataInterface
    value {dataInterface=&self:NiagaraDataInterfaceSkeletalMesh_0}
   }
  }
 ]
}
