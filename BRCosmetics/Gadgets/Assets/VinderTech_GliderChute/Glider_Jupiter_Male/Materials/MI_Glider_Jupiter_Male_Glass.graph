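# Simple glass material for the Jupiter Male glider cosmetic, built as a base Material despite its MI_ prefix.
# It uses Default Lit shading with no translucency setting or textures, and is flagged for skeletal meshes, morph targets and mesh deformers.
# The graph holds only a dark brown vector parameter and three scalar constants, with no wiring to material outputs shown.
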Material /BRCosmetics/Gadgets/Assets/VinderTech_GliderChute/Glider_Jupiter_Male/Materials/MI_Glider_Jupiter_Male_Glass
graph {
 properties {
  bUsedWithMeshDeformer=true
  bUsedWithMorphTargets=true
  bUsedWithSkeletalMesh=true
  pixelDepthOffsetMode=PDOM_Legacy
  usedShadingModels=Default Lit
 }
 nodes [
  Constant() {r=0.1}
  Constant()
  Constant() {r=0.5}
  VectorParameter() {defaultValue=<rgba 0.0296 0.0185 0.008 1>}
 ]
}
