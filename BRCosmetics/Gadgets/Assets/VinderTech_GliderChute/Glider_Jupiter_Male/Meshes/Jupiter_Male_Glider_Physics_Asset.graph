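# Physics asset for the Jupiter Male glider skeletal mesh, a VinderTech GliderChute cosmetic in BRCosmetics.
# Capsule bodies cover the handles, turret, mirrored wings, mounts, tails and turbines, with two overlapping capsules on Body_Offset.
# Constraints form a tree from Body_Offset through turret, wing and tail chains, all with identical swing and twist limit values.
# Handles_Base has a body but no constraint.
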PhysicsAsset /BRCosmetics/Gadgets/Assets/VinderTech_GliderChute/Glider_Jupiter_Male/Meshes/Jupiter_Male_Glider_Physics_Asset
bodies {
 Body_Offset [
  {shapeType=Capsule, length=224.5982, radius=64.0012, center=<-0.278 15.7382 35.2734>, rotation=<rot 0 179.874 -74.4436>}
  {shapeType=Capsule, length=148.5, radius=179.5, shapeName=Body_Offset_capsule, center=<0 29.8486 8.1079>, rotation=<rot 0 0 74.8032>}
 ]
 Handles_Base [{shapeType=Capsule, length=47.6027, radius=22.2765, center=<-0.0201 7.3739 -24.9542>, rotation=<rot 0 89.9887 -89.954>}]
 mount_01_l [{shapeType=Capsule, length=23.2323, radius=13.1877, center=<-19.2833 1.3615 -1.6151>, rotation=<rot 0 -98.8902 -82.5325>}]
 mount_01_r [{shapeType=Capsule, length=23.2323, radius=13.1877, center=<19.2833 -1.3614 1.6151>, rotation=<rot 0 -98.8902 -82.5325>}]
 tail_main_l [{shapeType=Capsule, length=48.7817, radius=24.4274, center=<-23.9288 -18.1227 -5.7724>, rotation=<rot 0 179.2623 -72.3432>}]
 tail_main_r [{shapeType=Capsule, length=48.7616, radius=24.3545, center=<23.8865 18.2958 5.6097>, rotation=<rot 0 179.593 -72.7596>}]
 tail_turbine_main_l [{shapeType=Capsule, length=30.8333, radius=16.8692, center=<-25.525 0.8095 -1.2535>, rotation=<rot 0 88.9803 -77.1092>}]
 tail_turbine_main_r [{shapeType=Capsule, length=29.4413, radius=11.7993, center=<27.0254 -0.754 -0.3781>, rotation=<rot 0 91.749 -85.4953>}]
 turret_03 [{shapeType=Capsule, length=15.1849, radius=14.279, center=<0.7238 -3.4848 0.0319>, rotation=<rot 0.1077 -90 0.0221>}]
 turret_04 [{shapeType=Capsule, length=19.4903, radius=3.5744, center=<19.3222 -0.0971 0.0002>, rotation=<rot 0 89.8463 -89.9992>}]
 wing_01_l [{shapeType=Capsule, length=62.6514, radius=27.7073, center=<-21.5973 -18.2493 -9.3651>, rotation=<rot 0 -179.2366 -74.249>}]
 wing_01_r [{shapeType=Capsule, length=62.7416, radius=28.0215, center=<19.9196 18.2617 9.6305>, rotation=<rot 0 -179.6192 -73.3048>}]
 wing_02_l [{shapeType=Capsule, length=41.482, radius=36.1825, center=<-41.4691 -0.6814 2.2968>, rotation=<rot 0 -87.7314 -159.6169>}]
 wing_02_r [{shapeType=Capsule, length=40.705, radius=35.9768, center=<42.6316 1.1616 -2.0291>, rotation=<rot 0 -87.9389 -163.8569>}]
 wing_main_l [{shapeType=Capsule, length=16.5424, radius=16.4359, center=<-8.8527 -0.216 0.1416>, rotation=<rot 0 0.2723 -147.17>}]
 wing_main_r [{shapeType=Capsule, length=16.5424, radius=16.4359, center=<8.8527 0.216 -0.1416>, rotation=<rot 0 0.2723 -147.17>}]
 wing_mount_inner_02_l [{shapeType=Capsule, length=32.2014, radius=7.249, center=<-0.7497 12.1378 0.0125>, rotation=<rot 0 2.3021 -89.964>}]
 wing_mount_inner_02_r [{shapeType=Capsule, length=32.2014, radius=7.249, center=<0.7497 -12.1378 -0.0123>, rotation=<rot 0 2.3021 -89.964>}]
 wing_mount_outer_02_l [{shapeType=Capsule, length=27.0175, radius=7.7282, center=<0.2118 7.0461 0.0358>, rotation=<rot 0 3.405 -89.9706>}]
 wing_mount_outer_02_r [{shapeType=Capsule, length=27.0175, radius=7.7282, center=<-0.2118 -7.0461 -0.0361>, rotation=<rot 0 3.405 -89.9706>}]
 wing_turbine_base_l [{shapeType=Capsule, length=14.8217, radius=12.9098, center=<-11.4458 -0.2823 0.5629>, rotation=<rot 0 -121.1118 -85.8067>}]
 wing_turbine_base_r [{shapeType=Capsule, length=29.3358, radius=15.0713, center=<-2.801 4.1255 -0.714>, rotation=<rot 0 -119.0888 -87.0661>}]
}
constraints [
 {bone1Name=turret_03, bone2Name=Body_Offset, swing1LimitDegrees=45, swing2LimitDegrees=45, twistLimitDegrees=45}
 {bone1Name=turret_04, bone2Name=turret_03, swing1LimitDegrees=45, swing2LimitDegrees=45, twistLimitDegrees=45}
 {bone1Name=wing_main_r, bone2Name=Body_Offset, swing1LimitDegrees=45, swing2LimitDegrees=45, twistLimitDegrees=45}
 {bone1Name=wing_01_r, bone2Name=wing_main_r, swing1LimitDegrees=45, swing2LimitDegrees=45, twistLimitDegrees=45}
 {bone1Name=wing_02_r, bone2Name=wing_01_r, swing1LimitDegrees=45, swing2LimitDegrees=45, twistLimitDegrees=45}
 {bone1Name=wing_mount_inner_02_r, bone2Name=wing_02_r, swing1LimitDegrees=45, swing2LimitDegrees=45, twistLimitDegrees=45}
 {bone1Name=wing_mount_outer_02_r, bone2Name=wing_02_r, swing1LimitDegrees=45, swing2LimitDegrees=45, twistLimitDegrees=45}
 {bone1Name=wing_turbine_base_r, bone2Name=wing_01_r, swing1LimitDegrees=45, swing2LimitDegrees=45, twistLimitDegrees=45}
 {bone1Name=mount_01_r, bone2Name=wing_main_r, swing1LimitDegrees=45, swing2LimitDegrees=45, twistLimitDegrees=45}
 {bone1Name=tail_main_r, bone2Name=Body_Offset, swing1LimitDegrees=45, swing2LimitDegrees=45, twistLimitDegrees=45}
 {bone1Name=tail_turbine_main_r, bone2Name=tail_main_r, swing1LimitDegrees=45, swing2LimitDegrees=45, twistLimitDegrees=45}
 {bone1Name=wing_main_l, bone2Name=Body_Offset, swing1LimitDegrees=45, swing2LimitDegrees=45, twistLimitDegrees=45}
 {bone1Name=wing_01_l, bone2Name=wing_main_l, swing1LimitDegrees=45, swing2LimitDegrees=45, twistLimitDegrees=45}
 {bone1Name=wing_02_l, bone2Name=wing_01_l, swing1LimitDegrees=45, swing2LimitDegrees=45, twistLimitDegrees=45}
 {bone1Name=wing_mount_inner_02_l, bone2Name=wing_02_l, swing1LimitDegrees=45, swing2LimitDegrees=45, twistLimitDegrees=45}
 {bone1Name=wing_mount_outer_02_l, bone2Name=wing_02_l, swing1LimitDegrees=45, swing2LimitDegrees=45, twistLimitDegrees=45}
 {bone1Name=wing_turbine_base_l, bone2Name=wing_01_l, swing1LimitDegrees=45, swing2LimitDegrees=45, twistLimitDegrees=45}
 {bone1Name=mount_01_l, bone2Name=wing_main_l, swing1LimitDegrees=45, swing2LimitDegrees=45, twistLimitDegrees=45}
 {bone1Name=tail_main_l, bone2Name=Body_Offset, swing1LimitDegrees=45, swing2LimitDegrees=45, twistLimitDegrees=45}
 {bone1Name=tail_turbine_main_l, bone2Name=tail_main_l, swing1LimitDegrees=45, swing2LimitDegrees=45, twistLimitDegrees=45}
]
